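# Parent material for the Clover vehicle cosmetic wheel, adding an animated, colored emissive effect on top of MF_WheelParent_Simple.
# The tread RGB texture is sampled panning and static: panned red is masked by static blue, boosted, then added to panned green.
# The result is tinted by VFX parameters and blended into the base emissive by vertex color red.
# The attributes then pass through vehicle damage and de-rez transition functions.
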Material /VehicleCosmetics/Wheels/Clover/Materials/M_WheelParent_Clover
graph {
 properties {
  bAutomaticallySetUsageInEditor=false
  bUseMaterialAttributes=true
  bUsedWithMeshDeformer=true
  bUsedWithSkeletalMesh=true
  pixelDepthOffsetMode=PDOM_Legacy
  previewMesh=/VehicleCosmetics/Wheels/Clover/Meshes/SK_Wheel_Clover
  usedShadingModels=Default Lit
 }
 nodes [
  n1 = MaterialFunctionCall() {materialFunction=&MF_WheelParent_Simple}
  n2 = GetMaterialAttributes(None=@n1.MaterialAttributes) {attributeGetTypes [#1]}
  n3 = TextureSample(UVs=Panner(Coordinate=TextureCoordinate()) {speedX=0.5}) {samplerSource=SSM_Wrap_WorldGroupSettings, texture=/VehicleCosmetics/Wheels/Clover/Textures/Wheel_Clover_Tread_RGB}
  n4 = TextureSample() {samplerSource=SSM_Wrap_WorldGroupSettings, texture=&Wheel_Clover_Tread_RGB}
  n5 = Add(
   A=Multiply(A=Multiply(A=@n3.R, B=@n4.B), B=ScalarParameter() {defaultValue=16, group=5_Special Effects, parameterName=VFX_Intensity})
   B=@n3.G
  )
  n6 = VectorParameter() {group=5_Special Effects, parameterName=VFX_Color, defaultValue=<rgba 0 0 1 1>}
  n7 = LinearInterpolate(
   A=@n2.Emissive Color
   B=Multiply(A=Multiply(A=@n5, B=@n6.RGB), B=ScalarParameter() {defaultValue=16, group=5_Special Effects, parameterName=VFX_Color_Intensity})
   Alpha=@n1.VertexColorR
  )
  MaterialFunctionCall(In=MaterialFunctionCall(Material=SetMaterialAttributes(MaterialAttributes=@n2.MaterialAttributes, Emissive Color=@n7) {attributeSetTypes [#1]}) {materialFunction=&MF_BaseVehicle_Damage}.Result) {materialFunction=&MF_VehicleDeRez_Transition}
 ]
}
