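# Translucent unlit skeletal-mesh material for the Clover wheel cosmetic, rendering animated clover and sparkle VFX.
# Three channels of a packed clover texture become big and small panning layers plus a rotating layer, each scaled by its color parameter's red channel.
# Bump-offset noise forms sparkles, with a mask from a curve texture scrolled along radial dot-product UVs.
# Assembled layers and a powered inverse mask drive opacity and emissive, the latter scaled by EmissiveMultiplier.
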Material /VehicleCosmetics/Wheels/Clover/Materials/M_Wheel_Clover_VFX
graph {
 properties {
  bAutomaticallySetUsageInEditor=false
  bUseMaterialAttributes=true
  bUsedWithMeshDeformer=true
  bUsedWithSkeletalMesh=true
  blendMode=BLEND_Translucent
  shadingModel=MSM_Unlit
  usedShadingModels=Unlit
 }
 nodes [
  n1 = TextureSample(UVs=BumpOffset(Coordinate=TextureCoordinate(), Height=Constant() {r=-8})) {samplerSource=SSM_Wrap_WorldGroupSettings, texture=/VehicleCosmetics/Wheels/_SharedTextures/FX/Noise_Generic_01_Pack}
  n2 = Clamp(None=@n1.G)
  NamedRerouteDeclaration(None=Multiply(A=Multiply(A=@n2) {constB=36}, B=@n1.B) {constB=36}) {name=Sparkle, nodeColor=<rgba 0 0.3372 0.5029 1>}
  n3 = Time()
  n4 = TextureSample(
   UVs=Panner(Coordinate=TextureCoordinate() {uTiling=0.6, vTiling=0.6}, Time=Add(A=@n3) {constB=0.35}) {speedX=-0.05, speedY=0.2}
   Tex=NamedRerouteUsage()
  ) {samplerSource=SSM_Wrap_WorldGroupSettings, texture=/VehicleCosmetics/Wheels/Clover/Textures/FXcloverRGB3}
  n5 = VectorParameter() {parameterName=BG color clovers BIG, defaultValue=<rgba 0.0452 0.2462 0 1>}
  NamedRerouteDeclaration(None=Multiply(A=Multiply(A=@n4.G) {constB=8}, B=AppendVector(A=@n5.R, B=@n5.R)) {constB=36}) {name=Clovers BIG, nodeColor=<rgba 0.2502 0.5029 0 1>}
  n6 = NamedRerouteUsage()
  n7 = NamedRerouteUsage()
  NamedRerouteDeclaration(None=Multiply(A=Multiply(A=@n6, B=@n7)) {constB=24}) {name=Sprakle Assembly, nodeColor=<rgba 0 0.3372 0.5029 1>}
  n8 = Power(Base=OneMinus(None=NamedRerouteUsage())) {constExponent=4}
  NamedRerouteDeclaration(None=@n8) {name=Power Mask, nodeColor=<rgba 1 0 0.4687 1>}
  n9 = NamedRerouteUsage()
  n10 = Multiply(A=@n8, B=NamedRerouteUsage())
  n11 = Multiply(A=@n10, B=@n10)
  n12 = Multiply(A=@n11) {constB=2}
  n13 = Add(A=Multiply(A=@n11, B=@n12) {constB=2}, B=@n12)
  n14 = Add(A=@n9, B=@n13)
  n15 = Multiply(A=Clamp(None=@n14), B=ScalarParameter() {defaultValue=2, parameterName=EmissiveMultiplier})
  NamedRerouteDeclaration(None=AppendVector(A=Add(A=@n15) {constB=0.025}, B=Constant())) {name=Emissive, nodeColor=<rgba 1 0 0.8203 1>}
  n16 = Add(A=@n9, B=@n14)
  NamedRerouteDeclaration(None=Multiply(A=Add(A=@n16, B=@n13), B=NamedRerouteUsage())) {name=Opacity, nodeColor=<rgba 0 1 0.8594 1>}
  NamedRerouteDeclaration(None=TextureObjectParameter() {parameterName=Clover VFX Tex 1, texture=&FXcloverRGB3}) {name=VFX Tex, nodeColor=<rgba 0 1 0.8125 1>}
  n17 = Time()
  n18 = TextureSample(
   UVs=Panner(Coordinate=TextureCoordinate() {uTiling=0.5, vTiling=0.5}, Time=Add(A=@n17) {constB=0.6}) {speedX=-0.15, speedY=0.2}
   Tex=NamedRerouteUsage()
  ) {samplerSource=SSM_Wrap_WorldGroupSettings, texture=&FXcloverRGB3}
  n19 = VectorParameter() {parameterName=BG color clovers Small, defaultValue=<rgba 0.0999 0.5841 0 1>}
  NamedRerouteDeclaration(None=Multiply(A=Multiply(A=@n18.R) {constB=2}, B=AppendVector(A=@n19.R, B=@n19.R)) {constB=36}) {name=Clovers Small, nodeColor=<rgba 0.2502 0.5029 0 1>}
  n20 = TextureSample(UVs=Rotator(Coordinate=TextureCoordinate() {uTiling=0.5, vTiling=0.5}) {centerX=0.05, centerY=0.01, speed=0.35}, Tex=NamedRerouteUsage()) {samplerSource=SSM_Wrap_WorldGroupSettings, texture=&FXcloverRGB3}
  n21 = VectorParameter() {parameterName=BG color clovers Rotating, defaultValue=<rgba 0.3278 1 0 1>}
  NamedRerouteDeclaration(None=Multiply(A=Multiply(A=@n20.B) {constB=0.4}, B=AppendVector(A=@n21.R, B=@n21.R)) {constB=36}) {name=Clovers Rotating, nodeColor=<rgba 0.2502 0.5029 0 1>}
  n22 = Add(A=NamedRerouteUsage(), B=NamedRerouteUsage())
  NamedRerouteDeclaration(None=Add(A=NamedRerouteUsage(), B=Clamp(None=@n22))) {name=Clovers Assembly, nodeColor=<rgba 0.2502 0.5029 0 1>}
  n23 = TextureSample(UVs=NamedRerouteUsage()) {samplerSource=SSM_Wrap_WorldGroupSettings, texture=/VehicleCosmetics/Wheels/_SharedTextures/FX/Curve_Rainbow}
  n24 = Power(Base=ComponentMask(None=Add(A=@n23.R, B=@n23.R)) {r=true}) {constExponent=0.2}
  NamedRerouteDeclaration(None=OneMinus(None=Clamp(None=@n24))) {name=Sparkle Mask, nodeColor=<rgba 0 0.3372 0.5029 1>}
  n25 = Add(A=TextureCoordinate()) {constB=-0.5}
  n26 = DotProduct(A=@n25, B=@n25)
  n27 = Clamp(None=Abs(None=@n26))
  NamedRerouteDeclaration(None=Panner(Coordinate=Multiply(A=@n27) {constB=2}) {speedX=-0.5}) {name=UV's For Sparkle Mask, nodeColor=<rgba 0 0.3372 0.5029 1>}
  NamedRerouteDeclaration(None=@n26) {name=Dot, nodeColor=<rgba 0 0.3372 0.5029 1>}
  SetMaterialAttributes!(MaterialAttributes, Emissive Color=NamedRerouteUsage(), Opacity=NamedRerouteUsage()) {attributeSetTypes [#1, #2]}
 ]
}
